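# Conversation database for the ZenCrash NPC, Carter Wu, the Peelians stunt coordinator.
# The entry greeting chooses contextual lines using first-ever meeting, match conversation count and outfit checks, with priority curves on some.
# It branches into links to the shared quests, basic services and Season 39 PerfectGift conversations, plus an "Ask About" menu.
# That menu offers Back or a chained Peelians monologue about the film's crew that returns to start.
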ConversationDatabase /NPCLibrary/NPCs/ZenCrash/NPCLibrary_Conversation_ZenCrash
entry_points [{tag=Conversation.Entry.ZenCrash, destinations [#1]}]
nodes [
 ConversationEntryPointNode_7 {
  id=#1
  class=ConversationEntryPointNode
  outputs [#2]
  properties {entryTag {tagName=Conversation.Entry.ZenCrash}}
 }
 Speech_C_4 {
  id=#2
  class=Speech_C
  outputs [#3, #4, #5, #6]
  properties {
   nodeName=""
   generalConfig {
    defaultMessage=I'm Carter Wu. I'm in charge of stunts for Peelians.
    contextualMessages [
     {
      message=I'm Carter Wu. I'm in charge of stunts for Peelians.
      contextRequirements [{requirement {class=FortControllerRequirement_IsFirstEverConversationWithNPC}}]
      priority {curve {rowName=Tandem.Speech.Priority.FirstEverInteraction}}
     }
     {
      message=Trust me. Nobody takes a hit like a stuntman.
      contextRequirements [{requirement {
         class=FortControllerRequirement_CheckNPCMatchConversationCount
         properties {nPCUniqueIdTag {tagName=AISpawnerData.Type.Specialist.Heavy.ZenCrash}}
        }}]
      priority {curve {curveTable=null, rowName=None}}
     }
     {
      message=Listen to your body. It knows more than you do.
      contextRequirements [{requirement {
         class=FortControllerRequirement_CheckNPCMatchConversationCount
         properties {nPCUniqueIdTag {tagName=AISpawnerData.Type.Specialist.Heavy.ZenCrash}}
        }}]
      priority {curve {curveTable=null, rowName=None}}
     }
     {
      message=You can make anything look real with the right preparation.
      contextRequirements [{requirement {
         class=FortControllerRequirement_CheckNPCMatchConversationCount
         properties {nPCUniqueIdTag {tagName=AISpawnerData.Type.Specialist.Heavy.ZenCrash}}
        }}]
      priority {curve {curveTable=null, rowName=None}}
     }
     {
      message="Nobody loves action sequences like we do, do they?"
      contextRequirements [{requirement {
         class=FortControllerRequirement_HasCID
         properties {allowedCIds [/BRCosmetics/Athena/Items/Cosmetics/Characters/Character_DerangedMile_Side]}
        }}]
      priority {curve {rowName=Tandem.Speech.Priority.OutfitReaction}}
     }
     {
      message="With a little care and a lot of preparation, the stunts in Peelians will be amazing."
      contextRequirements [{requirement {class=FortControllerRequirement_HasCID, properties {allowedCIds [/BRCosmetics/Athena/Items/Cosmetics/Characters/Character_FloralMane_Sand]}}}]
      priority {curve {rowName=Tandem.Speech.Priority.OutfitReaction}}
     }
     {
      message=You're a hard worker. I'll make an action star out of you in no time.
      contextRequirements [{requirement {class=FortControllerRequirement_HasCID, properties {allowedCIds [/BRCosmetics/Athena/Items/Cosmetics/Characters/Character_RowLiaison_Cafe]}}}]
      priority {curve {rowName=Tandem.Speech.Priority.OutfitReaction}}
     }
     {
      message=You're a natural at slipping... But we'll have to work on your landing.
      contextRequirements [{requirement {
         class=FortControllerRequirement_HasCID
         properties {allowedCIds [
           &Character_BananaPhilosopher
           &Character_BananaAdventure
           /BRCosmetics/Athena/Items/Cosmetics/Characters/Character_BananaCake
           &CID_349_Athena_Commando_M_Banana
           &CID_701_Athena_Commando_M_BananaAgent
           &CID_814_Athena_Commando_M_BananaSummer
           /BRCosmetics/Athena/Items/Cosmetics/Characters/CID_A_045_Athena_Commando_M_BananaLeader
           &CID_A_323_Athena_Commando_M_BananaWinter
           /BRCosmetics/Athena/Items/Cosmetics/Characters/CID_TBD_Athena_Commando_M_Banana_CINE
          ]}
        }}]
      priority {curve {rowName=Tandem.Speech.Priority.OutfitReaction}}
     }
     {
      message="No matter how careful you are, every great stunt is a little dangerous."
      contextRequirements [{requirement {
         class=FortControllerRequirement_CheckNPCMatchConversationCount
         properties {nPCUniqueIdTag {tagName=AISpawnerData.Type.Specialist.Heavy.ZenCrash}}
        }}]
      priority {curve {curveTable=null, rowName=None}}
     }
     {
      message=A good stuntman listens to his gut... and mine says that danger's on its way.
      contextRequirements [{requirement {
         class=FortControllerRequirement_CheckNPCMatchConversationCount
         properties {nPCUniqueIdTag {tagName=AISpawnerData.Type.Specialist.Heavy.ZenCrash}}
        }}]
      priority {curve {curveTable=null, rowName=None}}
      weight {value=2}
     }
    ]
   }
  }
 }
 ConversationLinkNode_2 {
  id=#6
  class=ConversationLinkNode
  properties {remoteEntryTag {tagName=Conversation.Entry.Season39.WF.PerfectGift}}
 }
 Speech_C_2 {
  id=#5
  class=Speech_C
  outputs [#7, #8]
  properties {nodeName="", softTaskIcon=&T_UI_NPC_question, generalConfig {defaultMessage=What do you need?}}
  sub_nodes [ConversationChoiceNode_0 {class=ConversationChoiceNode, properties {defaultChoiceDisplayText=Ask About...}}]
 }
 Speech_C_1 {
  id=#8
  class=Speech_C
  outputs [#9]
  properties {
   nodeName=""
   softTaskIcon=&T_UI_NPC_question
   generalConfig {defaultMessage="Peelians is just another job, honestly."}
  }
  sub_nodes [ConversationChoiceNode_41 {class=ConversationChoiceNode, properties {defaultChoiceDisplayText=Peelians}}]
 }
 Speech_C_6 {
  id=#9
  class=Speech_C
  outputs [#10]
  properties {
   nodeName=""
   softTaskIcon=&T_UI_NPC_question
   generalConfig {defaultMessage="I always love working with Kingston. He's passionate, respectful, a direct communicator..."}
  }
  sub_nodes [ConversationChoiceNode_43 {class=ConversationChoiceNode, properties {defaultChoiceDisplayText=Continue...}}]
 }
 Speech_C_8 {
  id=#10
  class=Speech_C
  outputs [#11]
  properties {
   nodeName=""
   softTaskIcon=&T_UI_NPC_question
   generalConfig {defaultMessage="Miles is great to work with, Cat's outside her comfort zone but working hard."}
  }
  sub_nodes [ConversationChoiceNode_44 {class=ConversationChoiceNode, properties {defaultChoiceDisplayText=Continue...}}]
 }
 Speech_C_9 {
  id=#11
  class=Speech_C
  outputs [#12]
  properties {
   nodeName=""
   softTaskIcon=&T_UI_NPC_question
   generalConfig {defaultMessage="And Peely... This town will chew him up and spit him out, but maybe he comes out of it a star."}
  }
  sub_nodes [ConversationChoiceNode_45 {class=ConversationChoiceNode, properties {defaultChoiceDisplayText=Continue...}}]
 }
 ReturnToStart_C_0 {
  id=#12
  class=ReturnToStart_C
  properties {softTaskIcon=&T_UI_NPC_Back}
  sub_nodes [ConversationChoiceNode_42 {class=ConversationChoiceNode, properties {defaultChoiceDisplayText=BACK}}]
 }
 Back_C_2 {
  id=#7
  class=Back_C
  sub_nodes [ConversationChoiceNode_40 {class=ConversationChoiceNode, properties {defaultChoiceDisplayText=BACK}}]
 }
 ConversationLinkNode_0 {id=#4, class=ConversationLinkNode, properties {remoteEntryTag {tagName=Conversation.Entry.BasicServices}}}
 ConversationLinkNode_1 {id=#3, class=ConversationLinkNode, properties {remoteEntryTag {tagName=Conversation.Entry.NPCLibrary.Quests}}}
]
